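; Fortnite art-tool turntable Blueprint that extends a multicapture multipass turntable with helpers to fill its capture list from editor-selected assets.
; Both functions read the selection through the TurntableEditorUtilityHelper component and wrap each asset in a MultipassTTSpawnStruct.
; One appends to MultiCaptureActorClasses; the other builds a local array and overwrites that list.
; Event graph and construction script add no logic of their own.
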(event EventBeginPlay
  (|CallParentFunction) ; This node is disabled and will not be called.\nDrag off pins to build functionality.
)

(event Collision|EventActorBeginOverlap
  (params (OtherActor :type Actor%20Object%20Reference)))

(event EventTick
  (params (DeltaSeconds :type Float%20%28single-precision%29)))

(fn AddSelectedAssetsToActorClasses
  (bind _assets (Class|TAGenericEditorWidgetInterface|GetSelectedAssets (Variables|Default|GetTurntableEditorUtilityHelper_Component)))
  (for _array_element _assets
    (Utilities|Array|Add (Variables|MultipassTurntable_MultiCapture|GetMultiCaptureActorClasses) (Utilities|Struct|MakeMultipassTTSpawnStruct "None" _array_element))))

(fn ReplaceActorClassesWithSelectedAssets
  (bind _assets (Class|TAGenericEditorWidgetInterface|GetSelectedAssets (Variables|Default|GetTurntableEditorUtilityHelper_Component)))
  (for _array_element _assets
    (Utilities|Array|Add (LocalVariables|Getl_temp) (Utilities|Struct|MakeMultipassTTSpawnStruct "None" _array_element)))
  (Variables|MultipassTurntable_MultiCapture|SetMultiCaptureActorClasses (LocalVariables|Getl_temp)))

(fn ConstructionScript
  (|CallParentFunction))
